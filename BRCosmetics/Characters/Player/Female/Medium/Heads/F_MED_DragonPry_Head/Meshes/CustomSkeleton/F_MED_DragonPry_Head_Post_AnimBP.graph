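# Post-process AnimBP for the female DragonPry cosmetic head that animates its hair straps.
# After copying the attached parent's pose, it additively tilts strap bones with Perlin-noise sway, weaker in the frontend.
# Chained spring-based AnimDynamics nodes then simulate seven bones per side, before a fixed hair_ALT_skin offset.
# The EventGraph also computes skydive, gravity, head-angle and QuietPeanuts backbling values, but this AnimGraph reads only TesselRotator_02 and its alpha.
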AnimBlueprint /BRCosmetics/Characters/Player/Female/Medium/Heads/F_MED_DragonPry_Head/Meshes/CustomSkeleton/F_MED_DragonPry_Head_Post_AnimBP
target_skeleton=&/Fortnite_Base_Head_DragonPry_Skeleton
graphs [
 AnimGraph {nodes [
   n1 = "Transform (Modify) Bone - Bone: base_hair_strap"(
    ComponentPose=Local To Component(LocalPose=Copy Pose From Mesh() {bUseAttachedParent=true})
    Alpha=1
    Translation="0,0,0"
    Rotation="45,0,0"
    Scale="1,1,1"
   ) {
    rotationMode=BMM_Additive
    rotationSpace=BCS_BoneSpace
    translationSpace=BCS_BoneSpace
    alphaBoolBlend {blendOption=Linear}
    alphaScaleBiasClamp {interpSpeedDecreasing=2, interpSpeedIncreasing=2}
    boneToModify {boneName=base_hair_strap}
    rotation {pitch=45}
   }
   n2 = Get TesselRotator_02() -> (TesselRotator_02:Rotator)
   n3 = Get Tassle 02 Alpha() -> (Tassle 02 Alpha:double)
   n4 = "Transform (Modify) Bone - Bone: base_hair_strap_a_l"(ComponentPose=@n1, Alpha=@n3, Translation="0,0,0", Rotation=@n2, Scale="1,1,1") {
    rotationMode=BMM_Additive
    rotationSpace=BCS_BoneSpace
    translationSpace=BCS_BoneSpace
    alphaBoolBlend {blendOption=Linear}
    alphaScaleBiasClamp {interpSpeedDecreasing=2, interpSpeedIncreasing=2}
    boneToModify {boneName=base_hair_strap_a_l}
   }
   n5 = "Transform (Modify) Bone - Bone: base_hair_strap_a_r"(ComponentPose=@n4, Alpha=@n3, Translation="0,0,0", Rotation=@n2, Scale="1,1,1") {
    rotationMode=BMM_Additive
    rotationSpace=BCS_BoneSpace
    translationSpace=BCS_BoneSpace
    alphaBoolBlend {blendOption=Linear}
    alphaScaleBiasClamp {bInterpResult=true, interpSpeedDecreasing=2, interpSpeedIncreasing=2}
    boneToModify {boneName=base_hair_strap_a_r}
   }
   n6 = "Transform (Modify) Bone - Bone: base_hair_strap"(ComponentPose=@n5, Alpha=1, Translation="0,0,0", Rotation="30,0,0", Scale="1,1,1") {
    rotationMode=BMM_Additive
    rotationSpace=BCS_BoneSpace
    translationSpace=BCS_BoneSpace
    alphaBoolBlend {blendOption=Linear}
    alphaScaleBiasClamp {bInterpResult=true, interpSpeedDecreasing=2, interpSpeedIncreasing=2}
    boneToModify {boneName=base_hair_strap}
    rotation {pitch=30}
   }
   n7 = Anim Dynamics - Bone: dyn_hair_strap_a_l!(ComponentPose=@n6, Alpha=1, GravityOverride:Vector="0,0,-980", ExternalForce="0,0,0") -> (Pose) {
    angularDampingOverride=0.9999
    angularSpringConstant=5000
    bAngularSpring=true
    bOverrideAngularDamping=true
    bUseGravityOverride=true
    gravityScale=0.35
    simulationSpace=RootRelative
    boundBone {boneName=dyn_hair_strap_a_l}
    componentAppliedLinearAccClamp=<2000 2000 2000>
    componentLinearAccScale=<1 1 1>
    componentLinearVelScale=<1 1 1>
    gravityOverride {z=-980}
    physicsBodyDefinitions [{
      boundBone {boneName=dyn_hair_strap_a_l}
      boxExtents=<20 30 30>
      constraintSetup {angularLimitsMax=<0 15 10>, angularLimitsMin=<0 -15 -10>, angularTarget=<1 0 0>}
      localJointOffset=<4 0 0>
     }]
   }
   n8 = Anim Dynamics - Bone: dyn_hair_strap_b_l!(ComponentPose=@n7, Alpha=1, GravityOverride:Vector="0,0,-980", ExternalForce="0,0,0") -> (Pose) {
    angularDampingOverride=0.9999
    angularSpringConstant=5000
    bAngularSpring=true
    bOverrideAngularDamping=true
    bUseGravityOverride=true
    gravityScale=0.35
    simulationSpace=RootRelative
    boundBone {boneName=dyn_hair_strap_b_l}
    componentAppliedLinearAccClamp=<2000 2000 2000>
    componentLinearAccScale=<1 1 1>
    componentLinearVelScale=<1 1 1>
    gravityOverride {z=-980}
    physicsBodyDefinitions [{
      boundBone {boneName=dyn_hair_strap_b_l}
      boxExtents=<20 30 30>
      constraintSetup {angularLimitsMax=<0 15 10>, angularLimitsMin=<0 -15 -10>, angularTarget=<1 0 0>}
      localJointOffset=<4 0 0>
     }]
   }
   n9 = Anim Dynamics - Bone: dyn_hair_strap_c_l!(ComponentPose=@n8, Alpha=1, GravityOverride:Vector="0,0,-980", ExternalForce="0,0,0") -> (Pose) {
    angularDampingOverride=0.9999
    angularSpringConstant=5000
    bAngularSpring=true
    bOverrideAngularDamping=true
    bUseGravityOverride=true
    gravityScale=0.35
    simulationSpace=RootRelative
    boundBone {boneName=dyn_hair_strap_c_l}
    componentAppliedLinearAccClamp=<2000 2000 2000>
    componentLinearAccScale=<1 1 1>
    componentLinearVelScale=<1 1 1>
    gravityOverride {z=-980}
    physicsBodyDefinitions [{
      boundBone {boneName=dyn_hair_strap_c_l}
      boxExtents=<20 30 30>
      constraintSetup {angularLimitsMax=<0 15 10>, angularLimitsMin=<0 -15 -10>, angularTarget=<1 0 0>}
      localJointOffset=<4 0 0>
     }]
   }
   n10 = Anim Dynamics - Bone: dyn_hair_strap_d_l!(ComponentPose=@n9, Alpha=1, GravityOverride:Vector="0,0,-980", ExternalForce="0,0,0") -> (Pose) {
    angularDampingOverride=0.9999
    angularSpringConstant=5000
    bAngularSpring=true
    bOverrideAngularDamping=true
    bUseGravityOverride=true
    gravityScale=0.35
    simulationSpace=RootRelative
    boundBone {boneName=dyn_hair_strap_d_l}
    componentAppliedLinearAccClamp=<2000 2000 2000>
    componentLinearAccScale=<1 1 1>
    componentLinearVelScale=<1 1 1>
    gravityOverride {z=-980}
    physicsBodyDefinitions [{
      boundBone {boneName=dyn_hair_strap_d_l}
      boxExtents=<20 30 30>
      constraintSetup {angularLimitsMax=<0 15 10>, angularLimitsMin=<0 -15 -10>, angularTarget=<1 0 0>}
      localJointOffset=<4 0 0>
     }]
   }
   n11 = Anim Dynamics - Bone: dyn_hair_strap_e_l!(ComponentPose=@n10, Alpha=1, GravityOverride:Vector="0,0,-980", ExternalForce="0,0,0") -> (Pose) {
    angularDampingOverride=0.9999
    angularSpringConstant=5000
    bAngularSpring=true
    bOverrideAngularDamping=true
    bUseGravityOverride=true
    gravityScale=0.35
    simulationSpace=RootRelative
    boundBone {boneName=dyn_hair_strap_e_l}
    componentAppliedLinearAccClamp=<2000 2000 2000>
    componentLinearAccScale=<1 1 1>
    componentLinearVelScale=<1 1 1>
    gravityOverride {z=-980}
    physicsBodyDefinitions [{
      boundBone {boneName=dyn_hair_strap_e_l}
      boxExtents=<20 30 30>
      constraintSetup {angularLimitsMax=<0 15 10>, angularLimitsMin=<0 -15 -10>, angularTarget=<1 0 0>}
      localJointOffset=<4 0 0>
     }]
   }
   n12 = Anim Dynamics - Bone: dyn_hair_strap_f_l!(ComponentPose=@n11, Alpha=1, GravityOverride:Vector="0,0,-980", ExternalForce="0,0,0") -> (Pose) {
    angularDampingOverride=0.9999
    angularSpringConstant=5000
    bAngularSpring=true
    bOverrideAngularDamping=true
    bUseGravityOverride=true
    gravityScale=0.35
    simulationSpace=RootRelative
    boundBone {boneName=dyn_hair_strap_f_l}
    componentAppliedLinearAccClamp=<2000 2000 2000>
    componentLinearAccScale=<1 1 1>
    componentLinearVelScale=<1 1 1>
    gravityOverride {z=-980}
    physicsBodyDefinitions [{
      boundBone {boneName=dyn_hair_strap_f_l}
      boxExtents=<20 30 30>
      constraintSetup {angularLimitsMax=<0 15 10>, angularLimitsMin=<0 -15 -10>, angularTarget=<1 0 0>}
      localJointOffset=<4 0 0>
     }]
   }
   n13 = Anim Dynamics - Bone: dyn_hair_strap_g_l!(ComponentPose=@n12, Alpha=1, GravityOverride:Vector="0,0,-980", ExternalForce="0,0,0") -> (Pose) {
    angularDampingOverride=0.9999
    angularSpringConstant=5000
    bAngularSpring=true
    bOverrideAngularDamping=true
    bUseGravityOverride=true
    gravityScale=0.35
    simulationSpace=RootRelative
    boundBone {boneName=dyn_hair_strap_g_l}
    componentAppliedLinearAccClamp=<2000 2000 2000>
    componentLinearAccScale=<1 1 1>
    componentLinearVelScale=<1 1 1>
    gravityOverride {z=-980}
    physicsBodyDefinitions [{
      boundBone {boneName=dyn_hair_strap_g_l}
      boxExtents=<20 30 30>
      constraintSetup {angularLimitsMax=<0 15 10>, angularLimitsMin=<0 -15 -10>, angularTarget=<1 0 0>}
      localJointOffset=<4 0 0>
     }]
   }
   n14 = Anim Dynamics - Bone: dyn_hair_strap_a_r!(ComponentPose=@n13, Alpha=1, GravityOverride:Vector="0,0,-980", ExternalForce="0,0,0") -> (Pose) {
    angularDampingOverride=0.9999
    angularSpringConstant=5000
    bAngularSpring=true
    bOverrideAngularDamping=true
    bUseGravityOverride=true
    gravityScale=0.35
    simulationSpace=RootRelative
    boundBone {boneName=dyn_hair_strap_a_r}
    componentAppliedLinearAccClamp=<2000 2000 2000>
    componentLinearAccScale=<1 1 1>
    componentLinearVelScale=<1 1 1>
    gravityOverride {z=-980}
    physicsBodyDefinitions [{
      boundBone {boneName=dyn_hair_strap_a_r}
      boxExtents=<20 30 30>
      constraintSetup {angularLimitsMax=<0 15 10>, angularLimitsMin=<0 -15 -10>, angularTarget=<1 0 0>}
      localJointOffset=<4 0 0>
     }]
   }
   n15 = Anim Dynamics - Bone: dyn_hair_strap_b_r!(ComponentPose=@n14, Alpha=1, GravityOverride:Vector="0,0,-980", ExternalForce="0,0,0") -> (Pose) {
    angularDampingOverride=0.9999
    angularSpringConstant=5000
    bAngularSpring=true
    bOverrideAngularDamping=true
    bUseGravityOverride=true
    gravityScale=0.35
    simulationSpace=RootRelative
    boundBone {boneName=dyn_hair_strap_b_r}
    componentAppliedLinearAccClamp=<2000 2000 2000>
    componentLinearAccScale=<1 1 1>
    componentLinearVelScale=<1 1 1>
    gravityOverride {z=-980}
    physicsBodyDefinitions [{
      boundBone {boneName=dyn_hair_strap_b_r}
      boxExtents=<20 30 30>
      constraintSetup {angularLimitsMax=<0 15 10>, angularLimitsMin=<0 -15 -10>, angularTarget=<1 0 0>}
      localJointOffset=<4 0 0>
     }]
   }
   n16 = Anim Dynamics - Bone: dyn_hair_strap_c_r!(ComponentPose=@n15, Alpha=1, GravityOverride:Vector="0,0,-980", ExternalForce="0,0,0") -> (Pose) {
    angularDampingOverride=0.9999
    angularSpringConstant=5000
    bAngularSpring=true
    bOverrideAngularDamping=true
    bUseGravityOverride=true
    gravityScale=0.35
    simulationSpace=RootRelative
    boundBone {boneName=dyn_hair_strap_c_r}
    componentAppliedLinearAccClamp=<2000 2000 2000>
    componentLinearAccScale=<1 1 1>
    componentLinearVelScale=<1 1 1>
    gravityOverride {z=-980}
    physicsBodyDefinitions [{
      boundBone {boneName=dyn_hair_strap_c_r}
      boxExtents=<20 30 30>
      constraintSetup {angularLimitsMax=<0 15 10>, angularLimitsMin=<0 -15 -10>, angularTarget=<1 0 0>}
      localJointOffset=<4 0 0>
     }]
   }
   n17 = Anim Dynamics - Bone: dyn_hair_strap_d_r!(ComponentPose=@n16, Alpha=1, GravityOverride:Vector="0,0,-980", ExternalForce="0,0,0") -> (Pose) {
    angularDampingOverride=0.9999
    angularSpringConstant=5000
    bAngularSpring=true
    bOverrideAngularDamping=true
    bUseGravityOverride=true
    gravityScale=0.35
    simulationSpace=RootRelative
    boundBone {boneName=dyn_hair_strap_d_r}
    componentAppliedLinearAccClamp=<2000 2000 2000>
    componentLinearAccScale=<1 1 1>
    componentLinearVelScale=<1 1 1>
    gravityOverride {z=-980}
    physicsBodyDefinitions [{
      boundBone {boneName=dyn_hair_strap_d_r}
      boxExtents=<20 30 30>
      constraintSetup {angularLimitsMax=<0 15 10>, angularLimitsMin=<0 -15 -10>, angularTarget=<1 0 0>}
      localJointOffset=<4 0 0>
     }]
   }
   n18 = Anim Dynamics - Bone: dyn_hair_strap_e_r!(ComponentPose=@n17, Alpha=1, GravityOverride:Vector="0,0,-980", ExternalForce="0,0,0") -> (Pose) {
    angularDampingOverride=0.9999
    angularSpringConstant=5000
    bAngularSpring=true
    bOverrideAngularDamping=true
    bUseGravityOverride=true
    gravityScale=0.35
    simulationSpace=RootRelative
    boundBone {boneName=dyn_hair_strap_e_r}
    componentAppliedLinearAccClamp=<2000 2000 2000>
    componentLinearAccScale=<1 1 1>
    componentLinearVelScale=<1 1 1>
    gravityOverride {z=-980}
    physicsBodyDefinitions [{
      boundBone {boneName=dyn_hair_strap_e_r}
      boxExtents=<20 30 30>
      constraintSetup {angularLimitsMax=<0 15 10>, angularLimitsMin=<0 -15 -10>, angularTarget=<1 0 0>}
      localJointOffset=<4 0 0>
     }]
   }
   n19 = Anim Dynamics - Bone: dyn_hair_strap_f_r!(ComponentPose=@n18, Alpha=1, GravityOverride:Vector="0,0,-980", ExternalForce="0,0,0") -> (Pose) {
    angularDampingOverride=0.9999
    angularSpringConstant=5000
    bAngularSpring=true
    bOverrideAngularDamping=true
    bUseGravityOverride=true
    gravityScale=0.35
    simulationSpace=RootRelative
    boundBone {boneName=dyn_hair_strap_f_r}
    componentAppliedLinearAccClamp=<2000 2000 2000>
    componentLinearAccScale=<1 1 1>
    componentLinearVelScale=<1 1 1>
    gravityOverride {z=-980}
    physicsBodyDefinitions [{
      boundBone {boneName=dyn_hair_strap_f_r}
      boxExtents=<20 30 30>
      constraintSetup {angularLimitsMax=<0 15 10>, angularLimitsMin=<0 -15 -10>, angularTarget=<1 0 0>}
      localJointOffset=<4 0 0>
     }]
   }
   n20 = Anim Dynamics - Bone: dyn_hair_strap_g_r!(ComponentPose=@n19, Alpha=1, GravityOverride:Vector="0,0,-980", ExternalForce="0,0,0") -> (Pose) {
    angularDampingOverride=0.9999
    angularSpringConstant=5000
    bAngularSpring=true
    bOverrideAngularDamping=true
    bUseGravityOverride=true
    gravityScale=0.35
    simulationSpace=RootRelative
    boundBone {boneName=dyn_hair_strap_g_r}
    componentAppliedLinearAccClamp=<2000 2000 2000>
    componentLinearAccScale=<1 1 1>
    componentLinearVelScale=<1 1 1>
    gravityOverride {z=-980}
    physicsBodyDefinitions [{
      boundBone {boneName=dyn_hair_strap_g_r}
      boxExtents=<20 30 30>
      constraintSetup {angularLimitsMax=<0 15 10>, angularLimitsMin=<0 -15 -10>, angularTarget=<1 0 0>}
      localJointOffset=<4 0 0>
     }]
   }
   n21 = "Transform (Modify) Bone - Bone: hair_ALT_skin"(ComponentPose=@n20, Alpha=1, Translation="5.421,1.967,0.464", Rotation="11.610,16.939,1.998", Scale="1,1,1") {
    rotationMode=BMM_Additive
    translationMode=BMM_Additive
    translationSpace=BCS_BoneSpace
    boneToModify {boneName=hair_ALT_skin}
    rotation=<rot 11.61 16.939 1.998>
    translation=<5.4207 1.9669 0.464>
   }
   Output Pose(Result=Component To Local(ComponentPose=@n21))
   Get TesselRotator_01() -> (TesselRotator_01:Rotator)
   Get Tassle 01 Alpha() -> (Tassle 01 Alpha:double)
  ]}
 EventGraph {nodes [
   Event BlueprintUpdateAnimation()
   > n1 = Sequence!(execute) -> (then_0, then_1, then_2:Exec, then_3:Exec, then_4:Exec)
   n2 = OR Boolean!(A=Get bIsSkydiveFloating(), B=Get bIsSkydiveDiveMode(), C:bool=Get bIsParachuteOpened()) -> (ReturnValue)
   > Set Is Skydiving Alpha(Is Skydiving Alpha:double="To Float (Boolean)"(InBool=@n2)) -> (Output_Get:double)
   > Set Is Skydive Bool Alpha(Is Skydive Bool Alpha:bool=@n2) -> (Output_Get:bool)
   n3 = "Gravity Override State Function (Parameter Asset) (Nativized)"(
    execute=@n1.then_1
    CharacterPartAnimInstance=Self-Reference()
    GameTimeInSeconds=GetGameTimeInSeconds()
    Parameters=/BRCosmetics/Characters/Player/Female/Medium/Heads/F_MED_DragonPry_Head/Meshes/F_MED_DragonPry_Head_Parameters
   )
   > Set GravityOverride(GravityOverride=@n3)
   > Set RBAN Settings Out(RBAN Settings Out=@n3.OutRigidBodyAnimNodeParameters)
   n4 = BreakTransform(InTransform=GetDeltaTransformFromRefPose(self=GetOwnerSkeletalMeshComponent(), BoneName=head, BaseName=spine_05))
   > Set HeadUP(HeadUP:double=MapRangeClamped(Value=BreakRotator(InRot=@n4.Rotation).Yaw, InRangeA=-30, InRangeB=-75)) -> (Output_Get:double)
   n5 = GetGameTimeInSeconds()
   n6 = float * float(A=@n5, B=1.125) {is=title K2Node_PromotableOperator}
   n7 = RandomRotatorFunction!(
    execute=@n1.then_2
    GameTimeInSeconds:double=@n6
    PerlinScale:double=1
    Frequency:Rotator="5,2.8,0"
    Amplitude:Rotator="30,15,0"
    Offset:Rotator="30,-1.8,0"
    TimeOffset:double=0
   ) -> (then, RandomRotator:Rotator)
   > Set TesselRotator_01(TesselRotator_01:Rotator=@n7) -> (Output_Get:Rotator)
   n8 = RandomRotatorFunction(execute=@n1.then_3, GameTimeInSeconds=@n6, PerlinScale=1, Frequency="4.5,1.6,0", Amplitude="30,15,0", Offset="45,15,0", TimeOffset=0)
   > Set TesselRotator_02(TesselRotator_02:Rotator=@n8) -> (Output_Get:Rotator)
   n9 = float * float(A=@n5, B=1.5)
   > n10 = RandomRotatorFunction(GameTimeInSeconds=@n9, PerlinScale=1, Frequency="6,2.6,0", Amplitude="30,15,0", Offset="35,15,0", TimeOffset=0)
   > Set TesselRotator_03(TesselRotator_03:Rotator=@n10) -> (Output_Get:Rotator)
   > n11 = RandomRotatorFunction(GameTimeInSeconds=@n9, PerlinScale=1, Frequency="5.5,2,0", Amplitude="30,15,0", Offset="35,15,0", TimeOffset=0)
   > Set TesselRotator_04(TesselRotator_04:Rotator=@n11) -> (Output_Get:Rotator)
   n12 = Branch(execute=@n1.then_4, Condition=Get bIsInFrontEnd())
   > Set Tassle 01 Alpha(Tassle 01 Alpha:double=0.25) -> (Output_Get:double)
   > Set Tassle 02 Alpha(Tassle 02 Alpha:double=0.425) -> (Output_Get:double)
   Set Tassle 01 Alpha(execute=@n12.else, Tassle 01 Alpha=0.5)
   > Set Tassle 02 Alpha(Tassle 02 Alpha=0.85)
   n13 = Event OnWeaponChanged()
   Get Dynamic Material Instance(execute=[@n13.then, @n19.then], SkeletalMeshComponent=GetPartSkeletalMeshComponent())
   > n14 = Sequence!(execute) -> (then_0, then_1)
   n15 = BackpackItemDefinitionHasAnyMetaTag!(Tags:Array of Gameplay Tag Structures=Make Array("[0]"=(TagName="Cosmetics.MetaTags.QuietPeanuts")), bRequiresEquipped:bool=false) -> (ReturnValue:bool)
   n16 = PickaxeItemDefinitionHasAnyMetaTag(Tags=Make Array("[0]"=(TagName="Cosmetics.MetaTags.QuietPeanuts")), bRequiresEquipped=true)
   > n17 = Branch(Condition=AND Boolean(A=NOT Boolean(A=@n16), B=@n15))
   > Set QuietPeanuts Backbling Collision Alpha(QuietPeanuts Backbling Collision Alpha:bool=true) -> (Output_Get:bool)
   Set QuietPeanuts Backbling Collision Alpha(execute=@n17.else, QuietPeanuts Backbling Collision Alpha=false)
   n18 = Branch(execute=@n14.then_1, Condition=@n15)
   > Set QuietPeanuts Backbling Alpha(QuietPeanuts Backbling Alpha:bool=true) -> (Output_Get:bool)
   Set QuietPeanuts Backbling Alpha(execute=@n18.else, QuietPeanuts Backbling Alpha=false)
   n19 = Event OnCharacterCustomizationLoaded()
   Get bIsSkyDiving()
  ]}
]
variables [
 AccScaleOverride {type=Vector}
 BwdMovement {type=number}
 currentBackpack {type=AthenaBackpackItemDefinition, default=null}
 Gravity Override {type=Vector}
 GravityOverride {type=Vector}
 HairRotator_01 {type=Rotator}
 HairRotator_02 {type=Rotator}
 HairRotator_03 {type=Rotator}
 HairRotator_04 {type=Rotator}
 HeadUP {type=number}
 Is Skydive Bool Alpha {type=boolean}
 Is Skydiving Alpha {type=number}
 isBackbling {type=boolean}
 isWearBackbling {type=boolean}
 isWearCape {type=boolean}
 Jumping {type=boolean}
 LookDown {type=number}
 LookUp {type=number}
 NotCrouching {type=boolean}
 playerPawn {type=FortPlayerPawn}
 QuietPeanuts Backbling Alpha {type=boolean}
 QuietPeanuts Backbling Collision Alpha {type=boolean}
 RBAN Settings Out {type=FortRigidBodyAnimNodeParameters}
 Tassle 01 Alpha {type=number}
 Tassle 02 Alpha {type=number}
 TesselRotator_01 {type=Rotator}
 TesselRotator_02 {type=Rotator}
 TesselRotator_03 {type=Rotator}
 TesselRotator_04 {type=Rotator}
 TrailControllerFakeVelocity {type=Vector}
]
